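# BigClockTicks is a SoundCue for the ticking of a big clock exterior prop, playing one tick sound per trigger.
# A Random node picks among five equally weighted tick wave variants.
# A Modulator at the root adds slight pitch and volume variation to avoid repetition.
# Playback uses the ambient props sound class and a dedicated big clock attenuation asset for distance falloff.
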SoundCue /Game/Sounds/Fort_Prop_Sounds/Exterior/BigClock/BigClockTicks
root=Modulator_0
nodes [
 {id=Modulator_0, children [Random_1], properties {pitchMax=1.01, pitchMin=0.99, volumeMin=0.9}}
 {id=Random_1, children [WavePlayer_3,WavePlayer_4,WavePlayer_5,WavePlayer_6,WavePlayer_2], properties {weights [1,1,1,1,1]}}
 {id=WavePlayer_2, properties {soundWaveAssetPtr=&/BigClock_Tick_05}}
 {id=WavePlayer_3, properties {soundWaveAssetPtr=&/BigClock_Tick_01}}
 {id=WavePlayer_4, properties {soundWaveAssetPtr=&/BigClock_Tick_02}}
 {id=WavePlayer_5, properties {soundWaveAssetPtr=&/BigClock_Tick_03}}
 {id=WavePlayer_6, properties {soundWaveAssetPtr=&/BigClock_Tick_04}}
]
properties {
 attenuationSettings=/Game/Sounds/Attenuation/Fort_BigClock_Att
 maxDistance=1600
 soundClassObject=&Fort_Ambient_Props_Athena
 volumeMultiplier=0.928
}
